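# Sound cue for the impact when the soccer ball toy is kicked.
# A Random node picks one of two equally weighted impact waves and feeds a Modulator that holds volume fixed, leaving pitch at its default variation.
# It uses shared ball-bounce attenuation and a concurrency override capping simultaneous plays.
# Notably, it is routed through the emote music sound class and has a send to the music audio-analysis submix.
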SoundCue /BRCosmetics/Sounds/Other/Toys/SoccerBall/SoccerBallToy_Kick_Hit_Cue
root=Modulator_0
nodes [
 {id=Modulator_0, children [Random_0], properties {volumeMax=1, volumeMin=1}}
 {id=Random_0, children [WavePlayer_0,WavePlayer_1], properties {weights [1,1]}}
 {id=WavePlayer_0, properties {soundWaveAssetPtr=&/SoccerBallToy_Kick_Impact_01}}
 {id=WavePlayer_1, properties {soundWaveAssetPtr=&/SoccerBallToy_Kick_Impact_02}}
]
properties {
 attenuationSettings=/Game/Sounds/Attenuation/Fort_Ball_Bounce_Athena_Att
 bOverrideConcurrency=true
 maxDistance=7000
 soundClassObject=&Fort_Music_Emote
 volumeMultiplier=1
 concurrencyOverrides {maxCount=3}
 soundSubmixSends [{minSendLevel=0, soundSubmix=&Fortnite_AudioAnalysis_Music}]
}
