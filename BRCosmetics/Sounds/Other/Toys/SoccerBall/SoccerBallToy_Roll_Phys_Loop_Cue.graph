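# Sound cue for the soccer ball toy's physics rolling loop, classed as a gadget sound with its own roll attenuation.
# A mixer layers a looping roll wave, given random pitch by a modulator, with a quieter repeating branch of roll-start sounds.
# That branch loops a randomly delayed pick from five start variations, adding irregular detail over the steady roll.
# The cue boosts overall volume and pitch and overrides concurrency.
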SoundCue /BRCosmetics/Sounds/Other/Toys/SoccerBall/SoccerBallToy_Roll_Phys_Loop_Cue
root=Mixer_0
nodes [
 {id=Delay_1, children [Random_1], properties {delayMax=1}}
 {id=Looping_0, children [Delay_1]}
 {id=Mixer_0, children [Looping_0,Modulator_0], properties {inputVolume [0.5,1]}}
 {id=Modulator_0, children [WavePlayer_0], properties {pitchMax=1.2, volumeMax=1, volumeMin=1}}
 {id=Random_1, children [WavePlayer_11, WavePlayer_12, WavePlayer_13, WavePlayer_14, WavePlayer_15], properties {weights [1,1,1,1,1]}}
 {id=WavePlayer_0, properties {bLooping=true, soundWaveAssetPtr=&/SoccerBallToy_Roll_Phys_Loop_01}}
 {id=WavePlayer_11, properties {soundWaveAssetPtr=&/SoccerBallToy_Roll_Phys_Start_01}}
 {id=WavePlayer_12, properties {soundWaveAssetPtr=&/SoccerBallToy_Roll_Phys_Start_02}}
 {id=WavePlayer_13, properties {soundWaveAssetPtr=&/SoccerBallToy_Roll_Phys_Start_03}}
 {id=WavePlayer_14, properties {soundWaveAssetPtr=&/SoccerBallToy_Roll_Phys_Start_04}}
 {id=WavePlayer_15, properties {soundWaveAssetPtr=&/SoccerBallToy_Roll_Phys_Start_05}}
]
properties {
 attenuationSettings=&/Fort_Toy_SoccerBall_Roll_Att
 bOverrideConcurrency=true
 maxDistance=7000
 pitchMultiplier=1.5
 soundClassObject=&Fort_Gadgets
 volumeMultiplier=1.5
}
